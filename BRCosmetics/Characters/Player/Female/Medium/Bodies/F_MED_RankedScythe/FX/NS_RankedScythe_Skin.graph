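# Cosmetic Niagara effect for the female RankedScythe character skin, attaching glows, rays and sparks to skeletal bones.
# Glow sprites and clavicle/head rays take gold, purple or silver from User.RankedProgression, with time-of-day dimming on rays.
# Blue-violet sparks, gated by User.UseTriangle, spawn both continuously on bones and by distance moved.
# User.bisHeadHidden drives sprite visibility tags; mesh aura emitters and a Cherry-skin ray emitter are disabled.
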NiagaraSystem /BRCosmetics/Characters/Player/Female/Medium/Bodies/F_MED_RankedScythe/FX/NS_RankedScythe_Skin
dependencies {
 usedDataInterfaces [
  &NiagaraDataInterfaceArrayFloat3
  &NiagaraDataInterfaceArrayFloat
  &NiagaraDataInterfaceCurve
  &NiagaraDataInterfaceColorCurve
  &NiagaraDataInterfaceArrayInt32
  &NiagaraDataInterfaceSkeletalMesh
 ]
 usedDynamicInputs [&MakeFloatFromVector2d]
 usedModules [
  &self:SystemSpawnScript.SystemScriptSource.SystemScriptGraph.NiagaraNodeAssignment_0.SetVariables_#1
  &SystemState
  &self:SystemSpawnScript.SystemScriptSource.SystemScriptGraph.NiagaraNodeAssignment_1.SetVariables_#2
  &self:SystemSpawnScript.SystemScriptSource.SystemScriptGraph.NiagaraNodeAssignment_2.SetVariables_#3
  &self:SystemSpawnScript.SystemScriptSource.SystemScriptGraph.NiagaraNodeAssignment_3.SetVariables_#4
 ]
}
emitters {
 AuraHead {
  bEnabled=false
  emitterUpdateScript [
   &EmitterState(
    Inactive Response=::NewEnumerator1
    Life Cycle Mode=::NewEnumerator1
    Loop Behavior=::NewEnumerator1
    Loop Duration Mode=::NewEnumerator1
    UseLoopDelay=false
   )
   &SpawnBurst_Instantaneous(Spawn Count=3, Use Spawn Probability=false, UseLoopCountLimit=false)
  ]
  particleSpawnScript [
   &PartitionParticles(Number Of Partitions=3)
   /SetVariables() {isSetParameters=true, module=SetVariables_#5}
   &InitializeParticle(
    Lifetime=2
    Mesh Scale=<1 1 0.5>
    Mesh Scale Mode=::NewEnumerator4
    Sprite Size Mode=::NewEnumerator3
    Uniform Sprite Size=50
    UsePositionOffset=false
    Write Material Random=true
   )
   &CameraOffset()
  ]
  particleUpdateScript [
   &SkeletalMeshLocation(
    Bone / Socket Index=6
    Bone Sampling Mode=::NewEnumerator4
    MaskSpawnGroup=false
    Orientation Sampling=::NewEnumerator1
    Sampled Position Offset=&SelectVectorFromArray(
     Direct Array Index=$Particles.PartitionParticles.Partition
     Vector Selection Array=DataInterface() {propertyValues {FloatData [<6 0 -5>, <10 0 0>, <6 0 5>]}}
    )
    Skeletal Mesh=$System.Skeletal Mesh
   )
   &ParticleState(Kill Particles When Lifetime Has Elapsed=false, Let Infinitely Lived Particles Die When Emitter Deactivates=true)
   &ScaleMeshSize(Scale Factor=&Multiply_VectorByFloat(
     Vector=&SelectVectorFromArray(
      Direct Array Index=$Particles.PartitionParticles.Partition
      Vector Selection Array=DataInterface() {propertyValues {FloatData [<0.3 1 2>, <0.4 1 1.8>, <0.3 1 1.6>]}}
     )
     Float=0.2
    ))
   &self:ColorSwitch(
    ColorInt:NiagaraInt32=$User.RankedProgression
    GoldColor:LinearColor=$System.GoldColor
    PurpleColor:LinearColor=$System.PurpleColor
    SilverColor:LinearColor=$System.SilverColor
   ) {module=ColorSwitch}
   &ScaleColor(
    Color Value To Scale=$Particles.Color
    Scale Alpha=&FloatFromCurve(
     FloatCurve=DataInterface() {propertyValues {Curve {keys [
         {interpMode=RCIM_Cubic, tangentMode=RCTM_User, time=0, value=1}
         {interpMode=RCIM_Cubic, tangentMode=RCTM_User, time=0.5, value=1}
         {tangentMode=RCTM_User, time=1, value=1}
        ]}}}
     Scale Curve=1.5
    )
    Scale RGB=&Multiply_VectorByFloat(Vector=<1 1 1>, Float=0.6)
    ScaleA=true
    ScaleRGB=true
   )
   &DynamicMaterialParameters(
    Index 0 Param 1=&SelectFloatFromArray(Direct Array Index=$Particles.PartitionParticles.Partition, Float Selection Array=DataInterface() {propertyValues {FloatData [0.1,0,0.1]}})
    Index 0 Param 2=&SelectFloatFromArray(Direct Array Index=$Particles.PartitionParticles.Partition, Float Selection Array=DataInterface() {propertyValues {FloatData [0,0.2,0.7]}})
    Module.Index0Param0Write=true
    Module.Index0Param1Write=true
    Module.Index0Param2Write=true
    Module.Index0Param3Write=true
   ) {module=DynamicMaterialParameters001}
  ]
  rendererClasses [&NiagaraMeshRendererProperties]
  renderers [{rendererClass=&NiagaraMeshRendererProperties, rendererIndex=0}]
 }
 AuraLeft {
  bEnabled=false
  emitterUpdateScript [
   &EmitterState(
    Inactive Response=::NewEnumerator1
    Life Cycle Mode=::NewEnumerator1
    Loop Behavior=::NewEnumerator1
    Loop Duration Mode=::NewEnumerator1
    UseLoopDelay=false
   )
   &SpawnBurst_Instantaneous(Spawn Count=3, Use Spawn Probability=false, UseLoopCountLimit=false)
  ]
  particleSpawnScript [
   &PartitionParticles(Number Of Partitions=3)
   /SetVariables(Particles.Random=&RandomRangeFloat(Minimum=0, Maximum=1)) {isSetParameters=true, module=SetVariables_#5}
   &InitializeParticle(
    Lifetime=2
    Mesh Scale=<1 1 0.5>
    Mesh Scale Mode=::NewEnumerator4
    Sprite Size Mode=::NewEnumerator3
    Uniform Sprite Size=50
    UsePositionOffset=false
    Write Material Random=true
   )
  ]
  particleUpdateScript [
   &SkeletalMeshLocation(
    Bone / Socket Index=$Particles.PartitionParticles.Partition
    Bone Sampling Mode=::NewEnumerator4
    MaskSpawnGroup=false
    Orientation Sampling=::NewEnumerator1
    Sampled Position Offset=&SelectVectorFromArray(
     Direct Array Index=$Particles.PartitionParticles.Partition
     Vector Selection Array=DataInterface() {propertyValues {FloatData [<3 0 0>, <8 0 0>, <10 0 -2>]}}
    )
    Skeletal Mesh=$System.Skeletal Mesh
   )
   &CameraOffset(Camera Offset Amount=&SelectFloatFromArray(Direct Array Index=$Particles.PartitionParticles.Partition, Float Selection Array=DataInterface() {propertyValues {FloatData [-10,0,0]}}))
   &ParticleState(Kill Particles When Lifetime Has Elapsed=false, Let Infinitely Lived Particles Die When Emitter Deactivates=true)
   &ScaleMeshSize(Scale Factor=&Multiply_VectorByFloat(
     Vector=&SelectVectorFromArray(
      Direct Array Index=$Particles.PartitionParticles.Partition
      Vector Selection Array=DataInterface() {propertyValues {FloatData [<0.65 1 4>, <0.35 1 3>, <0.45 1 2>]}}
     )
     Float=0.2
    ))
   &ScaleMeshSize(Scale Factor=<1 1 1.1>) {module=ScaleMeshSize001}
   &self:ColorSwitch(
    ColorInt:NiagaraInt32=$User.RankedProgression
    GoldColor:LinearColor=$System.GoldColor
    PurpleColor:LinearColor=$System.PurpleColor
    SilverColor:LinearColor=$System.SilverColor
   ) {module=ColorSwitch}
   &ScaleColor(
    Color Value To Scale=$Particles.Color
    Scale Alpha=&FloatFromCurve(
     FloatCurve=DataInterface() {propertyValues {Curve {keys [
         {interpMode=RCIM_Cubic, tangentMode=RCTM_User, time=0, value=1}
         {interpMode=RCIM_Cubic, tangentMode=RCTM_User, time=0.5, value=1}
         {tangentMode=RCTM_User, time=1, value=1}
        ]}}}
     Scale Curve=1.5
    )
    Scale RGB=&Multiply_VectorByFloat(Vector=<1 1 1>, Float=0.6)
    ScaleA=true
    ScaleRGB=true
   )
   &DynamicMaterialParameters(
    Index 0 Param 1=&SelectFloatFromArray(Direct Array Index=$Particles.PartitionParticles.Partition, Float Selection Array=DataInterface() {propertyValues {FloatData [0.1,0.2,0]}})
    Index 0 Param 2=&SelectFloatFromArray(Direct Array Index=$Particles.PartitionParticles.Partition, Float Selection Array=DataInterface() {propertyValues {FloatData [0,0.25,0.8]}})
    Module.Index0Param0Write=true
    Module.Index0Param1Write=true
    Module.Index0Param2Write=true
    Module.Index0Param3Write=true
   )
  ]
  rendererClasses [&NiagaraMeshRendererProperties]
  renderers [{rendererClass=&NiagaraMeshRendererProperties, rendererIndex=0}]
 }
 AuraRight {
  bEnabled=false
  emitterUpdateScript [
   &EmitterState(
    Inactive Response=::NewEnumerator1
    Life Cycle Mode=::NewEnumerator1
    Loop Behavior=::NewEnumerator1
    Loop Duration Mode=::NewEnumerator1
    UseLoopDelay=false
   )
   &SpawnBurst_Instantaneous(Spawn Count=3, Use Spawn Probability=false, UseLoopCountLimit=false)
  ]
  particleSpawnScript [
   &PartitionParticles(Number Of Partitions=3)
   /SetVariables() {isSetParameters=true, module=SetVariables_#5}
   &InitializeParticle(
    Lifetime=2
    Mesh Scale=<1 1 0.5>
    Mesh Scale Mode=::NewEnumerator4
    Sprite Size Mode=::NewEnumerator3
    Uniform Sprite Size=50
    UsePositionOffset=false
    Write Material Random=true
   )
  ]
  particleUpdateScript [
   &SkeletalMeshLocation(
    Bone / Socket Index=&Add_Integer(A=$Particles.PartitionParticles.Partition, B=3)
    Bone Sampling Mode=::NewEnumerator4
    MaskSpawnGroup=false
    Orientation Sampling=::NewEnumerator1
    Sampled Position Offset=&SelectVectorFromArray(
     Direct Array Index=$Particles.PartitionParticles.Partition
     Vector Selection Array=DataInterface() {propertyValues {FloatData [<-6 0 0>, <-9 0 0>, <-10 0 3>]}}
    )
    Skeletal Mesh=$System.Skeletal Mesh
   )
   &CameraOffset(Camera Offset Amount=&SelectFloatFromArray(Direct Array Index=$Particles.PartitionParticles.Partition, Float Selection Array=DataInterface() {propertyValues {FloatData [-10,0,0]}})) {module=CameraOffset001}
   &ParticleState(Kill Particles When Lifetime Has Elapsed=false, Let Infinitely Lived Particles Die When Emitter Deactivates=true)
   &ScaleMeshSize(Scale Factor=&Multiply_VectorByFloat(
     Vector=&SelectVectorFromArray(
      Direct Array Index=$Particles.PartitionParticles.Partition
      Vector Selection Array=DataInterface() {propertyValues {FloatData [<0.55 1 4>, <0.55 1 2>, <0.35 1 1.5>]}}
     )
     Float=0.2
    ))
   &ScaleMeshSize(Scale Factor=<1 1 1.2>) {module=ScaleMeshSize001}
   &self:ColorSwitch(
    ColorInt:NiagaraInt32=$User.RankedProgression
    GoldColor:LinearColor=$System.GoldColor
    PurpleColor:LinearColor=$System.PurpleColor
    SilverColor:LinearColor=$System.SilverColor
   ) {module=ColorSwitch}
   &ScaleColor(
    Color Value To Scale=$Particles.Color
    Scale Alpha=&FloatFromCurve(
     FloatCurve=DataInterface() {propertyValues {Curve {keys [
         {interpMode=RCIM_Cubic, tangentMode=RCTM_User, time=0, value=1}
         {interpMode=RCIM_Cubic, tangentMode=RCTM_User, time=0.5, value=1}
         {tangentMode=RCTM_User, time=1, value=1}
        ]}}}
     Scale Curve=1.5
    )
    Scale RGB=&Multiply_VectorByFloat(Vector=<1 1 1>, Float=0.6)
    ScaleA=true
    ScaleRGB=true
   )
   &DynamicMaterialParameters(
    Index 0 Param 1=&SelectFloatFromArray(Direct Array Index=$Particles.PartitionParticles.Partition, Float Selection Array=DataInterface() {propertyValues {FloatData [0,0.1,0.1]}})
    Index 0 Param 2=&SelectFloatFromArray(Direct Array Index=$Particles.PartitionParticles.Partition, Float Selection Array=DataInterface() {propertyValues {FloatData [0,0.4,0.6]}})
    Module.Index0Param0Write=true
    Module.Index0Param1Write=true
    Module.Index0Param2Write=true
    Module.Index0Param3Write=true
   ) {module=DynamicMaterialParameters001}
  ]
  rendererClasses [&NiagaraMeshRendererProperties]
  renderers [{rendererClass=&NiagaraMeshRendererProperties, rendererIndex=0}]
 }
 Glow {
  emitterUpdateScript [
   &EmitterState(
    Inactive Response=::NewEnumerator1
    Life Cycle Mode=::NewEnumerator1
    Loop Behavior=::NewEnumerator1
    Loop Duration Mode=::NewEnumerator1
    UseLoopDelay=false
   )
   &SpawnBurst_Instantaneous(Spawn Count=3, Use Spawn Probability=false, UseLoopCountLimit=false)
  ]
  particleSpawnScript [
   &PartitionParticles(Number Of Partitions=3)
   /SetVariables() {isSetParameters=true, module=SetVariables_#5}
   &InitializeParticle(
    Lifetime=2
    Mesh Scale=<1 1 0.5>
    Mesh Scale Mode=::NewEnumerator4
    Sprite Size Mode=::NewEnumerator3
    Uniform Sprite Size=25
    UsePositionOffset=false
    Write Material Random=true
   )
   &CameraOffset(Camera Offset Amount=5)
  ]
  particleUpdateScript [
   &SkeletalMeshLocation(
    Bone / Socket Index=&SelectIntFromArray(Direct Array Index=$Particles.PartitionParticles.Partition, Int Selection Array=DataInterface() {propertyValues {IntData [0,3,6]}})
    Bone Sampling Mode=::NewEnumerator4
    MaskSpawnGroup=false
    Orientation Sampling=::NewEnumerator1
    Sampled Position Offset=&SelectVectorFromArray(
     Direct Array Index=$Particles.PartitionParticles.Partition
     Vector Selection Array=DataInterface() {propertyValues {FloatData [<7 0 5>, <-7 0 -5>, <13 0 0>]}}
    )
    Skeletal Mesh=$System.Skeletal Mesh
   )
   &ParticleState(Kill Particles When Lifetime Has Elapsed=false, Let Infinitely Lived Particles Die When Emitter Deactivates=true)
   &self:ColorSwitch(
    ColorInt:NiagaraInt32=$User.RankedProgression
    GoldColor:LinearColor=$System.GoldColor
    PurpleColor:LinearColor=$System.PurpleColor
    SilverColor:LinearColor=$System.SilverColor
   ) {module=ColorSwitch}
   &ScaleColor(
    Color Value To Scale=$Particles.Color
    Scale Alpha=&FloatFromCurve(
     FloatCurve=DataInterface() {propertyValues {Curve {keys [{interpMode=RCIM_Cubic, tangentMode=RCTM_User, time=0.5, value=0}, {tangentMode=RCTM_User, time=1, value=1}]}}}
     Scale Curve=0.3
    )
    Scale RGB=&Multiply_VectorByFloat(Vector=<1 1 1>, Float=0.6)
    ScaleA=true
    ScaleRGB=true
   )
   &DynamicMaterialParameters(
    Index 0 Param 1=&SelectFloatFromArray(Direct Array Index=$Particles.PartitionParticles.Partition, Float Selection Array=DataInterface() {propertyValues {FloatData [0,0.1,0.1]}})
    Index 0 Param 2=&SelectFloatFromArray(Direct Array Index=$Particles.PartitionParticles.Partition, Float Selection Array=DataInterface() {propertyValues {FloatData [0,0.4,0.6]}})
    Module.Index0Param0Write=true
    Module.Index0Param1Write=true
    Module.Index0Param2Write=true
    Module.Index0Param3Write=true
   ) {module=DynamicMaterialParameters001}
   /SetVariables(Particles.VisibilityTag=&MakeCustomIntFromBool(Bool=$User.bisHeadHidden, True INT=1, False INT=0)) {isSetParameters=true, module=SetVariables_#6}
  ]
  rendererClasses [&NiagaraSpriteRendererProperties]
  renderers [{rendererClass=&NiagaraSpriteRendererProperties}]
 }
 NE_Cherry_VerticalRays001 {
  bEnabled=false
  emitterUpdateScript [
   &EmitterState(Life Cycle Mode=::NewEnumerator1, Loop Duration=1, UseLoopDelay=false)
   &SpawnRate(SpawnRate=16, Use Spawn Probability=false)
  ]
  particleSpawnScript [
   &FN_SetSpawnGroup(Number of Spawn Groups=3)
   /Niagara/Modules/Spawn/Initialization/InitializeParticle(
    Color=<rgba 1 0.5855 0.1372 0.5>
    Lifetime=3
    Module.Write Color=true
    Module.Write Lifetime=true
    Module.Write Mass=true
    Module.Write Material Random=true
    Module.Write Position=false
    Module.Write Scale=false
    Module.Write SpriteRotation=false
    Module.Write SpriteSize=true
    Position=&SimulationPosition()
    Sprite Size=&MaskVector2dBySpawnGroup(
     Masked Vector2D=&RandomRangeVector2D(Minimum=<1 30>, Maximum=<15 40>)
     Passthrough Vector2D=&RandomRangeVector2D(Minimum=<1 25>, Maximum=<5 35>)
     Spawn Group Mask=2
    )
   )
   &DynamicMaterialParameters(
    Index 0 Param 1=&RandomRangeFloat(Minimum=-0.1, Maximum=0.1)
    Module.Param0WriteEnabled=true
    Module.Param1WriteEnabled=false
    Module.Param2WriteEnabled=false
    Module.Param3WriteEnabled=false
   )
  ]
  particleUpdateScript [
   &SkeletalMeshLocation(
    Bone / Socket Index=$Particles.SpawnGroup
    Bone Sampling Mode=::NewEnumerator4
    Inherit Velocity=false
    MaskSpawnGroup=false
    Sampled Position Offset=&MakeVector(
     X=/Game/Effects/Niagara/DynamicInputs/FN_PickFloatFromSpawnGroupCutoff(
      Over Cutoff Float:NiagaraFloat=&RandomRangeFloat(Minimum=0, Maximum=0)
      Under Cutoff Float:NiagaraFloat=&FN_PickFloatFromSpawnGroupCutoff(
       Over Cutoff Float:NiagaraFloat=&RandomRangeFloat(Minimum=7, Maximum=14)
       Under Cutoff Float:NiagaraFloat=&RandomRangeFloat(Minimum=-7, Maximum=-14)
       SpawnGroup:NiagaraInt32=$Particles.SpawnGroup
       Cutoff:NiagaraInt32=1
      )
      SpawnGroup:NiagaraInt32=$Particles.SpawnGroup
      Cutoff:NiagaraInt32=2
     )
     Y=&FN_PickFloatFromSpawnGroupCutoff(
      Over Cutoff Float:NiagaraFloat=&RandomRangeFloat(Minimum=-3, Maximum=6)
      Under Cutoff Float:NiagaraFloat=&FN_PickFloatFromSpawnGroupCutoff(
       Over Cutoff Float:NiagaraFloat=&RandomRangeFloat(Minimum=-3, Maximum=3)
       Under Cutoff Float:NiagaraFloat=&RandomRangeFloat(Minimum=-3, Maximum=3)
       SpawnGroup:NiagaraInt32=$Particles.SpawnGroup
       Cutoff:NiagaraInt32=1
      )
      SpawnGroup:NiagaraInt32=$Particles.SpawnGroup
      Cutoff:NiagaraInt32=2
     )
     Z=&FN_PickFloatFromSpawnGroupCutoff(
      Over Cutoff Float:NiagaraFloat=&RandomRangeFloat(Minimum=-6, Maximum=6)
      Under Cutoff Float:NiagaraFloat=&FN_PickFloatFromSpawnGroupCutoff(
       Over Cutoff Float:NiagaraFloat=&RandomRangeFloat(Minimum=-8, Maximum=-5)
       Under Cutoff Float:NiagaraFloat=&RandomRangeFloat(Minimum=5, Maximum=8)
       SpawnGroup:NiagaraInt32=$Particles.SpawnGroup
       Cutoff:NiagaraInt32=1
      )
      SpawnGroup:NiagaraInt32=$Particles.SpawnGroup
      Cutoff:NiagaraInt32=2
     )
    )
    Sampled Position Offset Coordinate Space:ENiagaraMeshTransforms=::NewEnumerator3
    Skeletal Mesh=DataInterface() {propertyValues {
      DefaultMesh=null
      ExcludeBoneName=root
      OptionalComponentClass=null
      PreviewMesh=/BRCosmetics/Characters/Player/Female/Medium/Bodies/F_MED_Cherry/Meshes/F_MED_Cherry
      SkinningMode=SkinOnTheFly
      SoftSourceActor=null
      SourceMode=Default
      UvSetIndex=0
      WholeMeshLOD=-1
      bExcludeBone=false
      bReadDeformedGeometry=true
      bRequireCurrentFrameData=true
      ComponentTags []
      FilteredBones [clavicle_r,clavicle_l,head]
      FilteredSockets []
      MeshUserParameter {parameter {name=None, typeDefHandle {registeredTypeIndex=1}, varData []}}
      SamplingRegions []
     }}
   )
   &AddVelocity(Velocity=&MakeVector(X=0, Y=0, Z=&RandomRangeFloat(Minimum=9, Maximum=15))) {enabled=false}
   &ParticleState()
   &SolveForcesAndVelocity(Clamp Velocity=false, Limit Acceleration=false)
   &ScaleColor(
    Scale Alpha=&FloatFromCurve(FloatCurve=DataInterface() {propertyValues {Curve {keys [
         {arriveTangent=5.1832, interpMode=RCIM_Cubic, leaveTangent=5.1832, tangentMode=RCTM_User, time=0, value=0}
         {arriveTangent=2.2426, interpMode=RCIM_Cubic, leaveTangent=2.2426, tangentMode=RCTM_User, time=0.133, value=0.764}
         {arriveTangent=-0.0043, interpMode=RCIM_Cubic, leaveTangent=-0.0043, time=0.5, value=1}
         {arriveTangent=-1.5406, interpMode=RCIM_Cubic, leaveTangent=-1.5406, tangentMode=RCTM_User, time=0.8415, value=0.7609}
         {arriveTangent=-3.9306, interpMode=RCIM_Cubic, leaveTangent=-3.9306, tangentMode=RCTM_User, time=1, value=0}
        ]}}})
    Scale RGB=&MakeVectorFromLinearColorRGB(Color=&ColorFromCurve(ColorCurve=DataInterface() {propertyValues {
        AlphaCurve {keys [<at 0 1 RCIM_Cubic>]}
        BlueCurve {keys [<at 0 1 RCIM_Cubic>,<at 1 0.8559 RCIM_Cubic>]}
        GreenCurve {keys [<at 0 1 RCIM_Cubic>,<at 1 0.8559 RCIM_Cubic>]}
        RedCurve {keys [<at 0 1 RCIM_Cubic>,<at 1 1 RCIM_Cubic>]}
       }}))
    ScaleA=true
    ScaleRGB=false
   )
   &ScaleSpriteSize(Scale Factor=&MakeVector2D(X=1, Y=&FloatFromCurve(FloatCurve=DataInterface() {propertyValues {Curve {keys [<at 0 0>,<at 1 1>]}}})))
   &Color(Color=&MaskLinearColorBySpawnGroup(
     Masked Linear Color=<rgba 1 0 0.2339 1>
     Passthrough Linear Color=&MaskLinearColorBySpawnGroup(Masked Linear Color=<rgba 0.1072 1 0 1>, Passthrough Linear Color=<rgba 0 0.449 1 1>, Spawn Group Mask=1)
     Spawn Group Mask=2
    )) {enabled=false}
  ]
  rendererClasses [&NiagaraSpriteRendererProperties]
  renderers [{rendererClass=&NiagaraSpriteRendererProperties}]
 }
 Sparks_World_Local {
  emitterUpdateScript [
   &EmitterState(Inactive Response=::NewEnumerator1, Life Cycle Mode=::NewEnumerator1, Loop Delay=0.04, Loop Duration=1, UseLoopDelay=false)
   &SpawnRate(
    Spawn Probability=&MakeCustomFloatFromBool(Bool=$User.UseTriangle, True Float=1, False Float=0)
    SpawnRate=&RandomRangeFloat(Minimum=4, Maximum=8)
    Use Spawn Probability=true
   )
  ]
  particleSpawnScript [
   &PartitionParticles()
   /Niagara/Modules/Spawn/Initialization/V2/InitializeParticle(
    Lifetime Max=1.5
    Lifetime Mode=::NewEnumerator1
    Sprite Rotation Mode=::NewEnumerator2
    Sprite Size Mode=::NewEnumerator1
    Uniform Sprite Size Max=6
    Uniform Sprite Size Min=4
    UsePositionOffset=false
    Write Material Random=true
   )
   &ShapeLocation(
    MaskSpawnGroup=false
    Override Seed=false
    Shape Origin=&VectorFromFloat(Value=0)
    Sphere Radius=3
    UseOwnerScale=true
   )
   &CameraOffset() {module=CameraOffset001}
   &AddVelocity(Rotation Coordinate Space=::World)
  ]
  particleUpdateScript [
   &SkeletalMeshLocation(
    Bone / Socket Index=&Add_Integer(A=$Particles.PartitionParticles.Partition, B=2)
    Bone Sampling Mode=::NewEnumerator4
    MaskSpawnGroup=false
    Orientation Sampling=::NewEnumerator1
    Sampled Position Offset=&AddVectorToPosition(
     Position=&AddVectorToPosition(
      Position=<0 0 0>
      Vector=&RandomRangeVector(
       Minimum=&SelectVectorFromArray(
        Direct Array Index=$Particles.PartitionParticles.Partition
        Vector Selection Array=DataInterface() {propertyValues {FloatData [<12 0 0>, <-13 0 0>]}}
       )
       Maximum=&SelectVectorFromArray(
        Direct Array Index=$Particles.PartitionParticles.Partition
        Vector Selection Array=DataInterface() {propertyValues {FloatData [<5 0 0>, <-6 1 1>]}}
       )
      )
     )
     Vector=$Particles.Initial.ShapeLocation.ShapeVector
    )
    Skeletal Mesh=$System.Skeletal Mesh
   )
   &ParticleState()
   &SpriteRotationRate(Rotation Rate=&RandomRangeFloat(Minimum=200, Maximum=-200))
   &SolveForcesAndVelocity(Clamp Velocity=false, Limit Acceleration=false)
   &Color(
    Color=&ColorFromCurve(ColorCurve=DataInterface() {propertyValues {
       AlphaCurve {keys [<at 0 1 RCIM_Cubic>,<at 0.9977 1 RCIM_Cubic>]}
       BlueCurve {keys [<at 0.0448 1>,<at 0.8471 1>]}
       GreenCurve {keys [<at 0.0448 0.4649>,<at 0.8471 0.0403>]}
       RedCurve {keys [<at 0.0448 0>,<at 0.8471 0.0864>]}
      }})
    Scale Alpha=&FloatFromCurve(
     FloatCurve=DataInterface() {propertyValues {Curve {keys [
         {interpMode=RCIM_Cubic, tangentMode=RCTM_User, time=-0.0063, value=0.0184}
         {arriveTangent=-0.8483, interpMode=RCIM_Cubic, leaveTangent=-0.8483, tangentMode=RCTM_User, time=0.1718, value=1.7147}
         {arriveTangent=-5.0978, interpMode=RCIM_Cubic, leaveTangent=-5.0978, tangentMode=RCTM_User, time=0.2818, value=1.2143}
         {arriveTangent=-0.2783, interpMode=RCIM_Cubic, leaveTangent=-0.2783, tangentMode=RCTM_User, time=0.4, value=0.7}
         {arriveTangent=0.3112, interpMode=RCIM_Cubic, leaveTangent=0.3112, tangentMode=RCTM_User, time=0.5407, value=1.3419}
         {arriveTangent=-3.7119, interpMode=RCIM_Cubic, leaveTangent=-3.7119, tangentMode=RCTM_User, time=0.7242, value=0.3749}
         {interpMode=RCIM_Cubic, tangentMode=RCTM_User, time=1.0011, value=0.002}
        ]}}}
     Scale Curve=0.8
    )
    Scale Color=&VectorFromFloat(Value=50)
   ) {module=Color001}
   &ScaleColor(
    Color Value To Scale=$Particles.Color
    Scale Alpha=5
    Scale RGB=&VectorFromFloat(Value=&FloatFromCurve(FloatCurve=DataInterface() {propertyValues {Curve {keys [<at 0 0>,<at 0.2 2.625>,<at 0.4 2.5>,<at 1 1>]}}}, Scale Curve=0.3))
    ScaleA=true
    ScaleRGB=true
   )
   &ScaleSpriteSize(Scale Factor=&Vector2DFromFloat(Value=&FloatFromCurve(FloatCurve=DataInterface() {propertyValues {Curve {keys [
          {arriveTangent=3.043, interpMode=RCIM_Cubic, leaveTangent=3.043, tangentMode=RCTM_User, time=0, value=0}
          {arriveTangent=0.099, interpMode=RCIM_Cubic, leaveTangent=0.099, tangentMode=RCTM_User, time=0.3, value=0.65}
          {arriveTangent=0.1511, interpMode=RCIM_Cubic, leaveTangent=0.1511, tangentMode=RCTM_User, time=0.5, value=0.7}
          {interpMode=RCIM_Cubic, leaveTangent=-2.5, time=1, value=0}
         ]}}})))
   /SetVariables(Particles.Color2=&MultiplyLinearColor(
     Linear Color=&MultiplyLinearColor(Linear Color=$Particles.Color, Scale Factor=&MakeVector4_FromFloat(Float=0.5))
     Scale Factor=&MakeVector4_FromFloat(Float=0.1)
    )) {isSetParameters=true, module=SetVariables_#7}
   /SetVariables(Particles.SpriteSize2=&Multiply_Vector2D_ByFloat(Float=3.5, Vector2D=$Particles.SpriteSize)) {isSetParameters=true, module=SetVariables_#8}
   /SetVariables(Particles.Position=&AddVectorToPosition(
     Position=$Particles.Position
     Vector=&TransformVector(
      Vector=&Lerp_Vector(
       A=&SelectVectorFromArray(
        Direct Array Index=$Particles.PartitionParticles.Partition
        Vector Selection Array=DataInterface() {propertyValues {FloatData [<0 0 5>, <0 0 5>]}}
       )
       B=&SelectVectorFromArray(
        Direct Array Index=$Particles.PartitionParticles.Partition
        Vector Selection Array=DataInterface() {propertyValues {FloatData [<0 0 20>, <0 0 20>]}}
       )
       Alpha=$Particles.NormalizedAge
      )
      Destination Space=::World
     )
    )) {isSetParameters=true}
   /SetVariables(Particles.VisibilityTag=&MakeCustomIntFromBool(Bool=$User.bisHeadHidden, True INT=1, False INT=0)) {isSetParameters=true, module=SetVariables_#6}
  ]
  rendererClasses [&NiagaraSpriteRendererProperties]
  renderers [
   {rendererClass=&NiagaraSpriteRendererProperties}
   {rendererClass=&NiagaraSpriteRendererProperties, rendererIndex=1}
  ]
 }
 Sparks_World_UnLocal {
  emitterUpdateScript [
   &EmitterState(Inactive Response=::NewEnumerator1, Life Cycle Mode=::NewEnumerator1, Loop Delay=0.04, Loop Duration=1, UseLoopDelay=false)
   &SpawnPerUnit(
    Max Movement Threshold=1500
    Module.Use Max Movement Threshold=true
    Module.Use Movement Tolerance=false
    Spawn Probability:NiagaraFloat=&MakeCustomFloatFromBool(Bool=$User.UseTriangle, True Float=0.5, False Float=0)
    Spawn Spacing=50
    Use Spawn Probability=true
   )
  ]
  particleSpawnScript [
   &PartitionParticles()
   &InitializeParticle(
    Lifetime Max=0.5
    Lifetime Min=0.3
    Lifetime Mode=::NewEnumerator1
    Sprite Rotation Mode=::NewEnumerator2
    Sprite Size Mode=::NewEnumerator1
    Uniform Sprite Size Max=5
    Uniform Sprite Size Min=4
    UsePositionOffset=false
    Write Material Random=true
   )
   &CameraOffset() {module=CameraOffset001}
   &SkeletalMeshLocation(
    Bone / Socket Index=&Add_Integer(A=$Particles.PartitionParticles.Partition, B=2)
    Bone Sampling Mode=::NewEnumerator4
    MaskSpawnGroup=false
    Orientation Sampling=::NewEnumerator1
    Sampled Position Offset=&AddVectorToPosition(
     Position=&AddVectorToPosition(
      Position=&Lerp_Vector(
       A=&SelectVectorFromArray(
        Direct Array Index=$Particles.PartitionParticles.Partition
        Vector Selection Array=DataInterface() {propertyValues {FloatData [<0 0 5>, <0 0 -5>]}}
       )
       B=&SelectVectorFromArray(
        Direct Array Index=$Particles.PartitionParticles.Partition
        Vector Selection Array=DataInterface() {propertyValues {FloatData [<0 0 20>, <0 0 -20>]}}
       )
       Alpha=&RandomRangeFloat(Minimum=0, Maximum=1)
      )
      Vector=&RandomRangeVector(
       Minimum=&SelectVectorFromArray(
        Direct Array Index=$Particles.PartitionParticles.Partition
        Vector Selection Array=DataInterface() {propertyValues {FloatData [<12 0 0>, <-13 0 0>]}}
       )
       Maximum=&SelectVectorFromArray(
        Direct Array Index=$Particles.PartitionParticles.Partition
        Vector Selection Array=DataInterface() {propertyValues {FloatData [<5 0 0>, <-7 1 1>]}}
       )
      )
     )
     Vector=<0 0 0>
    )
    Skeletal Mesh=$System.Skeletal Mesh
   )
   &ShapeLocation(MaskSpawnGroup=false, Override Seed=false, Sphere Radius=7, UseOwnerScale=true)
  ]
  particleUpdateScript [
   &ParticleState()
   &SpriteRotationRate(Rotation Rate=&RandomRangeFloat(Minimum=200, Maximum=-200))
   &InheritSourceMovement(Applied Position Scale=<0.7 0.7 0.7>, ApplyForce=false, ApplyPosition=true, ApplyPositionFalloff=false)
   &SolveForcesAndVelocity(Clamp Velocity=false, Limit Acceleration=false)
   &Color(
    Color=&ColorFromCurve(ColorCurve=DataInterface() {propertyValues {
       AlphaCurve {keys [<at 0 1 RCIM_Cubic>,<at 0.9977 1 RCIM_Cubic>]}
       BlueCurve {keys [<at 0.0448 1>,<at 0.8471 1>]}
       GreenCurve {keys [<at 0.0448 0.4649>,<at 0.8471 0.0403>]}
       RedCurve {keys [<at 0.0448 0>,<at 0.8471 0.0864>]}
      }})
    Scale Alpha=&FloatFromCurve(
     FloatCurve=DataInterface() {propertyValues {Curve {keys [
         {interpMode=RCIM_Cubic, tangentMode=RCTM_User, time=-0.0063, value=0.0184}
         {arriveTangent=-0.8483, interpMode=RCIM_Cubic, leaveTangent=-0.8483, tangentMode=RCTM_User, time=0.1718, value=1.7147}
         {arriveTangent=-5.0978, interpMode=RCIM_Cubic, leaveTangent=-5.0978, tangentMode=RCTM_User, time=0.2818, value=1.2143}
         {arriveTangent=-0.2783, interpMode=RCIM_Cubic, leaveTangent=-0.2783, tangentMode=RCTM_User, time=0.4, value=0.7}
         {arriveTangent=0.3112, interpMode=RCIM_Cubic, leaveTangent=0.3112, tangentMode=RCTM_User, time=0.5407, value=1.3419}
         {arriveTangent=-3.7119, interpMode=RCIM_Cubic, leaveTangent=-3.7119, tangentMode=RCTM_User, time=0.7242, value=0.3749}
         {interpMode=RCIM_Cubic, tangentMode=RCTM_User, time=1.0011, value=0.002}
        ]}}}
     Scale Curve=0.8
    )
    Scale Color=&VectorFromFloat(Value=50)
   ) {module=Color001}
   &ScaleColor(
    Color Value To Scale=$Particles.Color
    Scale Alpha=5
    Scale RGB=&VectorFromFloat(Value=&FloatFromCurve(FloatCurve=DataInterface() {propertyValues {Curve {keys [<at 0 0>,<at 0.2 2.625>,<at 0.4 2.5>,<at 1 1>]}}}, Scale Curve=0.3))
    ScaleA=true
    ScaleRGB=true
   )
   &ScaleSpriteSize(Scale Factor=&Vector2DFromFloat(Value=&FloatFromCurve(FloatCurve=DataInterface() {propertyValues {Curve {keys [
          {arriveTangent=3.043, interpMode=RCIM_Cubic, leaveTangent=3.043, tangentMode=RCTM_User, time=0, value=0}
          {arriveTangent=0.099, interpMode=RCIM_Cubic, leaveTangent=0.099, tangentMode=RCTM_User, time=0.3, value=0.65}
          {interpMode=RCIM_Cubic, leaveTangent=-2.5, time=1, value=0}
         ]}}})))
   /SetVariables(Particles.Color2=&MultiplyLinearColor(
     Linear Color=&MultiplyLinearColor(Linear Color=$Particles.Color, Scale Factor=&MakeVector4_FromFloat(Float=0.5))
     Scale Factor=&MakeVector4_FromFloat(Float=0.1)
    )) {isSetParameters=true, module=SetVariables_#7}
   /SetVariables(Particles.SpriteSize2=&Multiply_Vector2D_ByFloat(Float=3.5, Vector2D=$Particles.SpriteSize)) {isSetParameters=true, module=SetVariables_#8}
   /SetVariables(Particles.VisibilityTag=&MakeCustomIntFromBool(Bool=$User.bisHeadHidden, True INT=1, False INT=0)) {isSetParameters=true, module=SetVariables_#6}
  ]
  rendererClasses [&NiagaraSpriteRendererProperties]
  renderers [
   {rendererClass=&NiagaraSpriteRendererProperties}
   {rendererClass=&NiagaraSpriteRendererProperties, rendererIndex=1}
  ]
 }
 VerticalRays {
  emitterUpdateScript [
   &EmitterState(Life Cycle Mode=::NewEnumerator1, Loop Duration=1, UseLoopDelay=false)
   &SpawnBurst_Instantaneous(Spawn Count=16, Use Spawn Probability=false, UseLoopCountLimit=false) {enabled=false}
   &SpawnRate(SpawnRate=10, Use Spawn Probability=false)
  ]
  particleSpawnScript [
   &FN_SetSpawnGroup(Number of Spawn Groups=3)
   &InitializeParticle(
    Lifetime Min=2
    Lifetime Mode=::NewEnumerator1
    Sprite Size=&MaskVector2dBySpawnGroup(
     Masked Vector2D=&RandomRangeVector2D(Minimum=<1 30>, Maximum=<15 40>)
     Passthrough Vector2D=&RandomRangeVector2D(Minimum=<1 25>, Maximum=<5 35>)
     Spawn Group Mask=2
    )
    Sprite Size Mode=::NewEnumerator4
    UsePositionOffset=false
    Write Material Random=true
   ) {module=InitializeParticle001}
   &DynamicMaterialParameters(
    Index 0 Param 1=&RandomRangeFloat(Minimum=-0.1, Maximum=0.1)
    Module.Index0Param0Write=true
    Module.Index0Param1Write=true
    Module.Index0Param2Write=true
    Module.Index0Param3Write=true
   ) {module=DynamicMaterialParameters001}
  ]
  particleUpdateScript [
   &SkeletalMeshLocation(
    Bone / Socket Index=$Particles.SpawnGroup
    Bone Sampling Mode=::NewEnumerator4
    MaskSpawnGroup=false
    Sampled Position Offset=&MakeVector(
     X=&FN_PickFloatFromSpawnGroupCutoff(
      Over Cutoff Float:NiagaraFloat=&RandomRangeFloat(Minimum=0, Maximum=0)
      Under Cutoff Float:NiagaraFloat=&FN_PickFloatFromSpawnGroupCutoff(
       Over Cutoff Float:NiagaraFloat=&RandomRangeFloat(Minimum=5, Maximum=13)
       Under Cutoff Float:NiagaraFloat=&RandomRangeFloat(Minimum=-5, Maximum=-14)
       SpawnGroup:NiagaraInt32=$Particles.SpawnGroup
       Cutoff:NiagaraInt32=1
      )
      SpawnGroup:NiagaraInt32=$Particles.SpawnGroup
      Cutoff:NiagaraInt32=2
     )
     Y=&FN_PickFloatFromSpawnGroupCutoff(
      Over Cutoff Float:NiagaraFloat=&RandomRangeFloat(Minimum=-3, Maximum=6)
      Under Cutoff Float:NiagaraFloat=&FN_PickFloatFromSpawnGroupCutoff(
       Over Cutoff Float:NiagaraFloat=&RandomRangeFloat(Minimum=-3, Maximum=3)
       Under Cutoff Float:NiagaraFloat=&RandomRangeFloat(Minimum=-3, Maximum=3)
       SpawnGroup:NiagaraInt32=$Particles.SpawnGroup
       Cutoff:NiagaraInt32=1
      )
      SpawnGroup:NiagaraInt32=$Particles.SpawnGroup
      Cutoff:NiagaraInt32=2
     )
     Z=&FN_PickFloatFromSpawnGroupCutoff(
      Over Cutoff Float:NiagaraFloat=&RandomRangeFloat(Minimum=-4, Maximum=4)
      Under Cutoff Float:NiagaraFloat=&FN_PickFloatFromSpawnGroupCutoff(
       Over Cutoff Float:NiagaraFloat=&RandomRangeFloat(Minimum=-5, Maximum=-3)
       Under Cutoff Float:NiagaraFloat=&RandomRangeFloat(Minimum=3, Maximum=5)
       SpawnGroup:NiagaraInt32=$Particles.SpawnGroup
       Cutoff:NiagaraInt32=1
      )
      SpawnGroup:NiagaraInt32=$Particles.SpawnGroup
      Cutoff:NiagaraInt32=2
     )
    )
    Skeletal Mesh=DataInterface() {propertyValues {
      DefaultMesh=null
      ExcludeBoneName=root
      OptionalComponentClass=null
      PreviewMesh=/BRCosmetics/Characters/Player/Female/Medium/Bodies/F_MED_RankedScythe/Meshes/F_MED_RankedScythe
      SkinningMode=SkinOnTheFly
      SoftSourceActor=null
      SourceMode=Default
      UvSetIndex=0
      WholeMeshLOD=-1
      bExcludeBone=false
      bReadDeformedGeometry=true
      bRequireCurrentFrameData=true
      ComponentTags []
      FilteredBones [clavicle_r,clavicle_l,head]
      FilteredSockets []
      MeshUserParameter {parameter {name=None, typeDefHandle {registeredTypeIndex=1}, varData []}}
      SamplingRegions []
     }}
   ) {module=SkeletalMeshLocation001}
   &ParticleState()
   &SolveForcesAndVelocity(Clamp Velocity=false, Limit Acceleration=false)
   &ScaleSpriteSize(
    Scale Factor=&MakeVector2D(X=1, Y=&FloatFromCurve(FloatCurve=DataInterface() {propertyValues {Curve {keys [<at 0 0>,<at 1 1>]}}}))
    Scale Sprite Size Mode=::NewEnumerator2
   ) {module=ScaleSpriteSize001}
   &self:ColorSwitch(
    ColorInt:NiagaraInt32=$User.RankedProgression
    GoldColor:LinearColor=$System.GoldColor
    PurpleColor:LinearColor=$System.PurpleColor
    SilverColor:LinearColor=$System.SilverColor
   ) {module=ColorSwitch}
   &ScaleColor(
    Color Value To Scale=$Particles.Color
    Scale Alpha=&FloatFromCurve(FloatCurve=DataInterface() {propertyValues {Curve {keys [
         {arriveTangent=2.3825, interpMode=RCIM_Cubic, leaveTangent=2.3825, tangentMode=RCTM_User, time=0, value=0}
         <at 0.5 1 RCIM_Cubic>
         {arriveTangent=-1.643, interpMode=RCIM_Cubic, leaveTangent=-1.643, tangentMode=RCTM_User, time=1, value=0}
        ]}}})
    Scale RGB=&Multiply_VectorByFloat(Vector=<1 1 1>, Float=&FNTimeOfDayScalar(Night Scale=0.7, Morning Scale=1, Evening Scale=1))
    ScaleA=true
    ScaleRGB=true
   )
   /SetVariables(Particles.VisibilityTag=&MakeCustomIntFromBool(Bool=$User.bisHeadHidden, True INT=1, False INT=0)) {isSetParameters=true, module=SetVariables_#6}
  ]
  rendererClasses [&NiagaraSpriteRendererProperties]
  renderers [{rendererClass=&NiagaraSpriteRendererProperties}]
 }
}
summary {
 systemName=NS_RankedScythe_Skin
 emitters [
  {bEnabled=false, emitterName=AuraLeft, rendererClasses [&NiagaraMeshRendererProperties]}
  {bEnabled=false, emitterName=AuraRight, rendererClasses [&NiagaraMeshRendererProperties]}
  {emitterName=Sparks_World_Local, rendererClasses [&NiagaraSpriteRendererProperties]}
  {emitterName=Sparks_World_UnLocal, rendererClasses [&NiagaraSpriteRendererProperties]}
  {bEnabled=false, emitterName=AuraHead, rendererClasses [&NiagaraMeshRendererProperties]}
  {emitterName=Glow, rendererClasses [&NiagaraSpriteRendererProperties]}
  {emitterName=VerticalRays, rendererClasses [&NiagaraSpriteRendererProperties]}
  {bEnabled=false, emitterName=NE_Cherry_VerticalRays001, rendererClasses [&NiagaraSpriteRendererProperties]}
 ]
 userVariables [
  User.bisHeadHidden {type=NiagaraBool}
  User.RankedProgression {type=NiagaraInt32}
  User.UseTriangle {type=NiagaraBool}
 ]
}
