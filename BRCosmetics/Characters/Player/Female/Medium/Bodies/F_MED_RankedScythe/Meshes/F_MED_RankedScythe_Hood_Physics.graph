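# Physics asset for the hood mesh of the female medium RankedScythe player cosmetic, with small capsule bodies on neck_02 and the left and right hood bones.
# Both hood_correction_base bones are constrained to neck_02 with free swing and twist.
# Each dyn_hood bone has locked swing and twist to the opposite side's correction base, so left pairs with right and right with left.
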PhysicsAsset /BRCosmetics/Characters/Player/Female/Medium/Bodies/F_MED_RankedScythe/Meshes/F_MED_RankedScythe_Hood_Physics
bodies {
 dyn_hood_l [{shapeType=Capsule, length=1, radius=2, shapeName=dyn_hood_l_capsule}]
 dyn_hood_r [{shapeType=Capsule, length=1, radius=2, shapeName=dyn_hood_r_capsule}]
 hood_correction_base_l [{shapeType=Capsule, length=1, radius=2, shapeName=hood_correction_base_l_capsule}]
 hood_correction_base_r [{shapeType=Capsule, length=1, radius=2, shapeName=hood_correction_base_r_capsule}]
 neck_02 [{shapeType=Capsule, length=0.505, radius=0.505, shapeName=neck_02_capsule, rotation=<rot 0 -90 0>}]
}
constraints [
 {
  bone1Name=dyn_hood_l
  bone2Name=hood_correction_base_r
  swing1LimitDegrees=5
  swing1Motion=Locked
  swing2LimitDegrees=5
  swing2Motion=Locked
  twistLimitDegrees=5
  twistMotion=Locked
 }
 {
  bone1Name=hood_correction_base_l
  bone2Name=neck_02
  swing1LimitDegrees=45
  swing1Motion=Free
  swing2LimitDegrees=45
  swing2Motion=Free
  twistLimitDegrees=45
  twistMotion=Free
 }
 {
  bone1Name=dyn_hood_r
  bone2Name=hood_correction_base_l
  swing1LimitDegrees=5
  swing1Motion=Locked
  swing2LimitDegrees=5
  swing2Motion=Locked
  twistLimitDegrees=5
  twistMotion=Locked
 }
 {
  bone1Name=hood_correction_base_r
  bone2Name=neck_02
  swing1LimitDegrees=45
  swing1Motion=Free
  swing2LimitDegrees=45
  swing2Motion=Free
  twistLimitDegrees=45
  twistMotion=Free
 }
]
