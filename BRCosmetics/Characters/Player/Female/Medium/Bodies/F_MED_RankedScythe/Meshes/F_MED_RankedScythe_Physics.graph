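# Physics asset for the F_MED_RankedScythe female body cosmetic, covering its dynamic coat and a dangling skull.
# Capsules on pelvis, legs and feet accompany six coat chains and the skull, all anchored to the pelvis with tight angular limits below each base.
# Free-angle constraints cross-link some neighbouring lower coat segments, tying panels together.
# A huge zero-length root capsule tops out just above the origin, apparently as a ground collider.
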PhysicsAsset /BRCosmetics/Characters/Player/Female/Medium/Bodies/F_MED_RankedScythe/Meshes/F_MED_RankedScythe_Physics
bodies {
 calf_l [{shapeType=Capsule, length=27.5, radius=8, shapeName=calf_l_capsule, center=<-21.8857 -2.7832 0>, rotation=<rot 0 89.9999 90>}]
 calf_r [{shapeType=Capsule, length=27.5, radius=8, shapeName=calf_r_capsule, center=<21.7111 1.7927 0>, rotation=<rot 0 89.9994 -90>}]
 dyn_coat_bck_l_01 [{shapeType=Capsule, length=13.5, radius=6.5, shapeName=dyn_coat_bck_l_01_capsule, center=<3.8739 0 3.182>, rotation=<rot 0 0 90>}]
 dyn_coat_bck_l_02 [{shapeType=Capsule, length=13.5, radius=6.5, shapeName=dyn_coat_bck_l_02_capsule, center=<3.8739 0 3.7424>, rotation=<rot 0 0 90>}]
 dyn_coat_bck_l_03 [{shapeType=Capsule, length=13.5, radius=6.5, shapeName=dyn_coat_bck_l_03_capsule, center=<0 0 5.6298>, rotation=<rot 0 0 90>}]
 dyn_coat_bck_l_base [{shapeType=Capsule, length=3, radius=1.5, shapeName=dyn_coat_bck_l_base_capsule, center=<0 0 3.4693>, rotation=<rot 0 0 90>}]
 dyn_coat_bck_r_01 [{shapeType=Capsule, length=13.5, radius=6.5, shapeName=dyn_coat_bck_r_01_capsule, center=<3.8739 0 3.182>, rotation=<rot 0 0 90>}]
 dyn_coat_bck_r_02 [{shapeType=Capsule, length=13.5, radius=6.5, shapeName=dyn_coat_bck_r_02_capsule, center=<3.8739 0 3.7424>, rotation=<rot 0 0 90>}]
 dyn_coat_bck_r_03 [{shapeType=Capsule, length=13.5, radius=6.5, shapeName=dyn_coat_bck_r_03_capsule, center=<0 0 5.6298>, rotation=<rot 0 0 90>}]
 dyn_coat_bck_r_base [{shapeType=Capsule, length=3, radius=1.5, shapeName=dyn_coat_bck_r_base_capsule, center=<0 0 3.4693>, rotation=<rot 0 0 90>}]
 dyn_coat_fr_l_01 [{shapeType=Capsule, length=13.5, radius=6.5, shapeName=dyn_coat_fr_l_01_capsule, center=<7.1713 0 6.101>, rotation=<rot 0 0 90>}]
 dyn_coat_fr_l_02 [{shapeType=Capsule, length=13.5, radius=6.5, shapeName=dyn_coat_fr_l_02_capsule, center=<3.8739 0 2.4781>, rotation=<rot 0 0 90>}]
 dyn_coat_fr_l_03 [{shapeType=Capsule, length=13.5, radius=6.5, shapeName=dyn_coat_fr_l_03_capsule, center=<0 0 3.0061>, rotation=<rot 0 0 90>}]
 dyn_coat_fr_l_base [{shapeType=Capsule, length=3, radius=1.5, shapeName=dyn_coat_fr_l_base_capsule, center=<0 0 3.4693>, rotation=<rot 0 0 90>}]
 dyn_coat_fr_r_01 [{shapeType=Capsule, length=5.5, radius=6.5, shapeName=dyn_coat_fr_r_01_capsule, center=<3.8739 0 3.2114>, rotation=<rot 0 0 90>}]
 dyn_coat_fr_r_02 [{shapeType=Capsule, length=8.5, radius=6.5, shapeName=dyn_coat_fr_r_02_capsule, center=<3.8739 2.3521 2.8432>, rotation=<rot 0 0 90>}]
 dyn_coat_fr_r_03 [{shapeType=Capsule, length=8, radius=6.5, shapeName=dyn_coat_fr_r_03_capsule, center=<0 2.3521 2.1324>, rotation=<rot 0 0 90>}]
 dyn_coat_fr_r_base [{shapeType=Capsule, length=3, radius=1.5, shapeName=dyn_coat_fr_r_base_capsule, center=<0 0 3.4693>, rotation=<rot 0 0 90>}]
 dyn_coat_side_l_01 [{shapeType=Capsule, length=13.5, radius=6.5, shapeName=dyn_coat_side_l_01_capsule, center=<3.8739 0 3.7424>, rotation=<rot 0 0 90>}]
 dyn_coat_side_l_02 [{shapeType=Capsule, length=13.5, radius=6.5, shapeName=dyn_coat_side_l_02_capsule, center=<0.0119 0 3.7424>, rotation=<rot 0 0 90>}]
 dyn_coat_side_l_03 [{shapeType=Capsule, length=13.5, radius=6.5, shapeName=dyn_coat_side_l_03_capsule, center=<0 0 5.6298>, rotation=<rot 0 0 90>}]
 dyn_coat_side_l_base [{shapeType=Capsule, length=3, radius=1.5, shapeName=dyn_coat_side_l_base_capsule, center=<0 0 3.4693>, rotation=<rot 0 0 90>}]
 dyn_coat_side_r_01 [{shapeType=Capsule, length=13.5, radius=6.5, shapeName=dyn_coat_side_r_01_capsule, center=<3.8739 0 3.7424>, rotation=<rot 0 0 90>}]
 dyn_coat_side_r_02 [{shapeType=Capsule, length=16, radius=6.5, shapeName=dyn_coat_side_r_02_capsule, center=<1.5984 0 6.211>, rotation=<rot 0 0 90>}]
 dyn_coat_side_r_03 [{shapeType=Capsule, length=16, radius=6.5, shapeName=dyn_coat_side_r_03_capsule, center=<-2.2755 0 6.1391>, rotation=<rot 0 0 90>}]
 dyn_coat_side_r_base [{shapeType=Capsule, length=3, radius=1.5, shapeName=dyn_coat_side_r_base_capsule, center=<0 0 3.4693>, rotation=<rot 0 0 90>}]
 dyn_skull [{shapeType=Capsule, length=3, radius=2.5, shapeName=dyn_skull_capsule, center=<-2.8321 1.9778 0>, rotation=<rot 90 0 0>}]
 dyn_skull_base [{shapeType=Capsule, length=0.505, radius=0.505, shapeName=dyn_skull_base_capsule, rotation=<rot 0 -90 0>}]
 foot_l [{shapeType=Capsule, length=4, radius=8, shapeName=foot_l_capsule, center=<-3.1667 5.4849 0>, rotation=<rot 0 29.9996 -90>}]
 foot_r [{shapeType=Capsule, length=4, radius=8, shapeName=foot_r_capsule, center=<2.0656 -3.5778 0>, rotation=<rot 0 29.9996 -90>}]
 pelvis [{shapeType=Capsule, length=19.5, radius=4.5, shapeName=pelvis_capsule, center=<0 2.4779 0>, rotation=<rot -89.3567 179.9997 180>}]
 root [{shapeType=Capsule, length=0, radius=289.5, shapeName=root_capsule, center=<0 0 -286.8952>}]
 thigh_l [{shapeType=Capsule, length=25.5, radius=9, shapeName=thigh_l_capsule, center=<-23.673 -0.2343 0>, rotation=<rot 90 180 -0.5669>}]
 thigh_r [{shapeType=Capsule, length=25, radius=9.5, shapeName=thigh_r_capsule, center=<23.6742 -0.0001 0>, rotation=<rot 90 -0.0002 0.0001>}]
}
constraints [
 {
  bone1Name=calf_l
  bone2Name=thigh_l
  swing1LimitDegrees=45
  swing1Motion=Free
  swing2LimitDegrees=45
  swing2Motion=Free
  twistLimitDegrees=45
  twistMotion=Free
 }
 {
  bone1Name=calf_r
  bone2Name=thigh_r
  swing1LimitDegrees=45
  swing1Motion=Free
  swing2LimitDegrees=45
  swing2Motion=Free
  twistLimitDegrees=45
  twistMotion=Free
 }
 {
  bone1Name=foot_r
  bone2Name=calf_r
  swing1LimitDegrees=45
  swing1Motion=Free
  swing2LimitDegrees=45
  swing2Motion=Free
  twistLimitDegrees=45
  twistMotion=Free
 }
 {
  bone1Name=foot_l
  bone2Name=calf_l
  swing1LimitDegrees=45
  swing1Motion=Free
  swing2LimitDegrees=45
  swing2Motion=Free
  twistLimitDegrees=45
  twistMotion=Free
 }
 {
  bone1Name=thigh_l
  bone2Name=pelvis
  swing1LimitDegrees=45
  swing1Motion=Free
  swing2LimitDegrees=45
  swing2Motion=Free
  twistLimitDegrees=45
  twistMotion=Free
 }
 {
  bone1Name=thigh_r
  bone2Name=pelvis
  swing1LimitDegrees=45
  swing1Motion=Free
  swing2LimitDegrees=45
  swing2Motion=Free
  twistLimitDegrees=45
  twistMotion=Free
 }
 {bone1Name=dyn_coat_side_r_03, bone2Name=dyn_coat_side_r_02, swing1LimitDegrees=5, swing2LimitDegrees=20, twistLimitDegrees=2}
 {
  bone1Name=dyn_coat_bck_l_base
  bone2Name=pelvis
  swing1LimitDegrees=45
  swing1Motion=Free
  swing2LimitDegrees=45
  swing2Motion=Free
  twistLimitDegrees=45
  twistMotion=Free
 }
 {bone1Name=dyn_coat_bck_l_01, bone2Name=dyn_coat_bck_l_base, swing1LimitDegrees=5, swing2LimitDegrees=20, twistLimitDegrees=2}
 {bone1Name=dyn_coat_bck_l_02, bone2Name=dyn_coat_bck_l_01, swing1LimitDegrees=5, swing2LimitDegrees=20, twistLimitDegrees=2}
 {bone1Name=dyn_coat_bck_l_03, bone2Name=dyn_coat_bck_l_02, swing1LimitDegrees=5, swing2LimitDegrees=20, twistLimitDegrees=2}
 {
  bone1Name=dyn_coat_bck_r_base
  bone2Name=pelvis
  swing1LimitDegrees=45
  swing1Motion=Free
  swing2LimitDegrees=45
  swing2Motion=Free
  twistLimitDegrees=45
  twistMotion=Free
 }
 {bone1Name=dyn_coat_bck_r_01, bone2Name=dyn_coat_bck_r_base, swing1LimitDegrees=5, swing2LimitDegrees=20, twistLimitDegrees=2}
 {bone1Name=dyn_coat_bck_r_02, bone2Name=dyn_coat_bck_r_01, swing1LimitDegrees=5, swing2LimitDegrees=20, twistLimitDegrees=2}
 {bone1Name=dyn_coat_bck_r_03, bone2Name=dyn_coat_bck_r_02, swing1LimitDegrees=5, swing2LimitDegrees=20, twistLimitDegrees=2}
 {
  bone1Name=dyn_coat_fr_l_base
  bone2Name=pelvis
  swing1LimitDegrees=45
  swing1Motion=Free
  swing2LimitDegrees=45
  swing2Motion=Free
  twistLimitDegrees=45
  twistMotion=Free
 }
 {bone1Name=dyn_coat_fr_l_01, bone2Name=dyn_coat_fr_l_base, swing1LimitDegrees=5, swing2LimitDegrees=20, twistLimitDegrees=2}
 {bone1Name=dyn_coat_fr_l_02, bone2Name=dyn_coat_fr_l_01, swing1LimitDegrees=5, swing2LimitDegrees=30, twistLimitDegrees=2}
 {bone1Name=dyn_coat_fr_l_03, bone2Name=dyn_coat_fr_l_02, swing1LimitDegrees=5, swing2LimitDegrees=20, twistLimitDegrees=2}
 {
  bone1Name=dyn_coat_fr_r_base
  bone2Name=pelvis
  swing1LimitDegrees=45
  swing1Motion=Free
  swing2LimitDegrees=45
  swing2Motion=Free
  twistLimitDegrees=45
  twistMotion=Free
 }
 {bone1Name=dyn_coat_fr_r_01, bone2Name=dyn_coat_fr_r_base, swing1LimitDegrees=5, swing2LimitDegrees=20, twistLimitDegrees=2}
 {bone1Name=dyn_coat_fr_r_02, bone2Name=dyn_coat_fr_r_01, swing1LimitDegrees=5, swing2LimitDegrees=40, twistLimitDegrees=2}
 {bone1Name=dyn_coat_fr_r_03, bone2Name=dyn_coat_fr_r_02, swing1LimitDegrees=5, swing2LimitDegrees=20, twistLimitDegrees=2}
 {
  bone1Name=dyn_coat_side_l_base
  bone2Name=pelvis
  swing1LimitDegrees=45
  swing1Motion=Free
  swing2LimitDegrees=45
  swing2Motion=Free
  twistLimitDegrees=45
  twistMotion=Free
 }
 {bone1Name=dyn_coat_side_l_01, bone2Name=dyn_coat_side_l_base, swing1LimitDegrees=5, swing2LimitDegrees=20, twistLimitDegrees=2}
 {bone1Name=dyn_coat_side_l_02, bone2Name=dyn_coat_side_l_01, swing1LimitDegrees=5, swing2LimitDegrees=40, twistLimitDegrees=2}
 {bone1Name=dyn_coat_side_l_03, bone2Name=dyn_coat_side_l_02, swing1LimitDegrees=5, swing2LimitDegrees=20, twistLimitDegrees=2}
 {
  bone1Name=dyn_coat_side_r_base
  bone2Name=pelvis
  swing1LimitDegrees=45
  swing1Motion=Free
  swing2LimitDegrees=45
  swing2Motion=Free
  twistLimitDegrees=45
  twistMotion=Free
 }
 {bone1Name=dyn_coat_side_r_01, bone2Name=dyn_coat_side_r_base, swing1LimitDegrees=5, swing2LimitDegrees=20, twistLimitDegrees=2}
 {bone1Name=dyn_coat_side_r_02, bone2Name=dyn_coat_side_r_01, swing1LimitDegrees=5, swing2LimitDegrees=40, twistLimitDegrees=2}
 {
  bone1Name=dyn_coat_side_l_03
  bone2Name=dyn_coat_fr_l_03
  swing1LimitDegrees=45
  swing1Motion=Free
  swing2LimitDegrees=45
  swing2Motion=Free
  twistLimitDegrees=45
  twistMotion=Free
 }
 {
  bone1Name=dyn_coat_side_r_03
  bone2Name=dyn_coat_bck_r_03
  swing1LimitDegrees=45
  swing1Motion=Free
  swing2LimitDegrees=45
  swing2Motion=Free
  twistLimitDegrees=45
  twistMotion=Free
 }
 {
  bone1Name=dyn_coat_bck_r_03
  bone2Name=dyn_coat_bck_l_03
  swing1LimitDegrees=45
  swing1Motion=Free
  swing2LimitDegrees=45
  swing2Motion=Free
  twistLimitDegrees=45
  twistMotion=Free
 }
 {
  bone1Name=dyn_coat_side_r_03
  bone2Name=dyn_coat_fr_r_03
  swing1LimitDegrees=45
  swing1Motion=Free
  swing2LimitDegrees=45
  swing2Motion=Free
  twistLimitDegrees=45
  twistMotion=Free
 }
 {
  bone1Name=dyn_coat_side_r_02
  bone2Name=dyn_coat_bck_r_02
  swing1LimitDegrees=45
  swing1Motion=Free
  swing2LimitDegrees=45
  swing2Motion=Free
  twistLimitDegrees=45
  twistMotion=Free
 }
 {
  bone1Name=dyn_coat_bck_r_02
  bone2Name=dyn_coat_bck_l_02
  swing1LimitDegrees=45
  swing1Motion=Free
  swing2LimitDegrees=45
  swing2Motion=Free
  twistLimitDegrees=45
  twistMotion=Free
 }
 {bone1Name=dyn_skull, bone2Name=dyn_skull_base, swing1LimitDegrees=40, swing2LimitDegrees=5, twistLimitDegrees=3}
 {
  bone1Name=dyn_skull_base
  bone2Name=pelvis
  swing1LimitDegrees=45
  swing1Motion=Free
  swing2LimitDegrees=45
  swing2Motion=Free
  twistLimitDegrees=45
  twistMotion=Free
 }
 {
  bone1Name=dyn_coat_side_l_03
  bone2Name=dyn_coat_bck_l_03
  swing1LimitDegrees=45
  swing1Motion=Free
  swing2LimitDegrees=45
  swing2Motion=Free
  twistLimitDegrees=45
  twistMotion=Free
 }
]
